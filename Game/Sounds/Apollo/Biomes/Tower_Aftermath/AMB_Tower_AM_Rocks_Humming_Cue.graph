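# Ambient sound cue for humming rocks in the Tower Aftermath biome, played as a looping outdoor emitter with medium prop attenuation.
# A mixer layers the rocks' humming loop with a quieter Zero Point tower base bed loop.
# A quality node routes every level to that same mixer, and a short randomized delay at the root offsets the start.
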SoundCue /Game/Sounds/Apollo/Biomes/Tower_Aftermath/AMB_Tower_AM_Rocks_Humming_Cue
root=Delay_0
nodes [
 {id=Delay_0, children [QualityLevel_1], properties {delayMax=0.3}}
 {id=Mixer_0, children [WavePlayer_3,WavePlayer_0], properties {inputVolume [1.25,0.8]}}
 {id=QualityLevel_1, children [Mixer_0,Mixer_0,Mixer_0,Mixer_0]}
 {id=WavePlayer_0, properties {bLooping=true, soundWaveAssetPtr=/Game/Sounds/Apollo/Biomes/ZeroPoint/AMB_Bed_ZP_S16_Tower_Base_Loop_01}}
 {
  id=WavePlayer_3
  properties {bLooping=true, soundWaveAssetPtr=&/Amb_Tower_Aftermath_Humming_Rocks_Loop}
 }
]
properties {
 attenuationSettings=/Game/Sounds/Attenuation/Fort_Amb_Emitter_Prop_Med_Att
 maxDistance=900
 soundClassObject=&Fort_Ambient_Outdoors_Athena
 volumeMultiplier=1.504
 preEffectBusSends [{}]
}
