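# Sound cue for the death of an alien flora (variant 02) in the Tower Aftermath biome.
# A Random node picks one of four equally weighted death samples and feeds a Modulator that varies pitch, with a lowered minimum, for variation.
# It plays in the impacts sound class at slightly reduced volume, using the small-object demolition destroy attenuation asset.
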SoundCue /Game/Sounds/Apollo/Biomes/Tower_Aftermath/Tower_Aftermath_Alien_Flora02_Death_Cue
root=Modulator_0
nodes [
 {id=Modulator_0, children [Random_0], properties {pitchMin=0.9}}
 {id=Random_0, children [WavePlayer_0,WavePlayer_1,WavePlayer_2,WavePlayer_3], properties {weights [1,1,1,1]}}
 {id=WavePlayer_0, properties {soundWaveAssetPtr=&/Tower_Aftermath_Alien_Flora02_Death_01}}
 {id=WavePlayer_1, properties {soundWaveAssetPtr=&/Tower_Aftermath_Alien_Flora02_Death_02}}
 {id=WavePlayer_2, properties {soundWaveAssetPtr=&/Tower_Aftermath_Alien_Flora02_Death_03}}
 {id=WavePlayer_3, properties {soundWaveAssetPtr=&/Tower_Aftermath_Alien_Flora02_Death_04}}
]
properties {
 attenuationSettings=/Game/Sounds/Attenuation/Fort_Demolition_Destroy_Small_Sized_Object_Att
 maxDistance=4000
 soundClassObject=&Fort_Impacts
 volumeMultiplier=0.8
}
