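# Sound cue for impacts on the second alien flora variant in the Tower Aftermath biome, classed under Fort_Impacts.
# A Random node picks one of five impact waves with equal weighting and feeds a Modulator, whose lowered pitch minimum adds variation between hits.
# The cue plays at reduced volume and borrows the shared small-object demolition hit attenuation for spatial falloff.
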SoundCue /Game/Sounds/Apollo/Biomes/Tower_Aftermath/Tower_Aftermath_Alien_Flora02_Impact_Cue
root=Modulator_0
nodes [
 {id=Modulator_0, children [Random_1], properties {pitchMin=0.9}}
 {id=Random_1, children [WavePlayer_4,WavePlayer_5,WavePlayer_6,WavePlayer_7,WavePlayer_8], properties {weights [1,1,1,1,1]}}
 {id=WavePlayer_4, properties {soundWaveAssetPtr=&/Tower_Aftermath_Alien_Flora02_Impact_01}}
 {id=WavePlayer_5, properties {soundWaveAssetPtr=&/Tower_Aftermath_Alien_Flora02_Impact_02}}
 {id=WavePlayer_6, properties {soundWaveAssetPtr=&/Tower_Aftermath_Alien_Flora02_Impact_03}}
 {id=WavePlayer_7, properties {soundWaveAssetPtr=&/Tower_Aftermath_Alien_Flora02_Impact_04}}
 {id=WavePlayer_8, properties {soundWaveAssetPtr=&/Tower_Aftermath_Alien_Flora02_Impact_05}}
]
properties {
 attenuationSettings=/Game/Sounds/Attenuation/Fort_Demolition_Hit_Small_Sized_Object_Att
 maxDistance=4000
 soundClassObject=&Fort_Impacts
 volumeMultiplier=0.72
}
