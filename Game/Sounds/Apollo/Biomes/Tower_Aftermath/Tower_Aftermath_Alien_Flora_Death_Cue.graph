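# Sound cue for the death of alien flora in the Tower Aftermath biome, layering vine impact and pain sounds.
# A quality-level node selects one of four mixers, ranging from randomized pools of four waves per layer down to a single wave each.
# Impacts get random pitch, and some pain pools include empty slots so that layer sometimes stays silent.
# It uses small-object destruction attenuation and a four-max object-impact concurrency set.
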SoundCue /Game/Sounds/Apollo/Biomes/Tower_Aftermath/Tower_Aftermath_Alien_Flora_Death_Cue
root=QualityLevel_0
nodes [
 {id=Mixer_0, children [Modulator_0,Modulator_1], properties {inputVolume [1,0.5]}}
 {id=Mixer_1, children [Modulator_3,Random_5], properties {inputVolume [1,0.5]}}
 {id=Mixer_2, children [Modulator_2,Random_4], properties {inputVolume [1,0.5]}}
 {id=Mixer_3, children [Modulator_5, WavePlayer_16], properties {inputVolume [1,1]}}
 {id=Modulator_0, children [Random_6], properties {pitchMax=1.07, pitchMin=0.92}}
 {id=Modulator_1, children [Random_3], properties {pitchMax=1.07, pitchMin=0.92}}
 {id=Modulator_2, children [Random_8], properties {pitchMax=1.07, pitchMin=0.92}}
 {id=Modulator_3, children [Random_7], properties {pitchMax=1.07, pitchMin=0.92}}
 {id=Modulator_5, children [WavePlayer_21], properties {pitchMax=1.1, pitchMin=0.9}}
 {id=QualityLevel_0, children [Mixer_0,Mixer_2,Mixer_1,Mixer_3]}
 {id=Random_3, children [WavePlayer_8, WavePlayer_9, WavePlayer_10, WavePlayer_11, null, null], properties {weights [1,1,1,1,1,1]}}
 {id=Random_4, children [WavePlayer_12, WavePlayer_13], properties {weights [1,1]}}
 {id=Random_5, children [WavePlayer_14, WavePlayer_15, null], properties {weights [1,1,1]}}
 {id=Random_6, children [WavePlayer_17, WavePlayer_18, WavePlayer_19, WavePlayer_20], properties {weights [1,1,1,1]}}
 {id=Random_7, children [WavePlayer_22, WavePlayer_23], properties {weights [1,1]}}
 {id=Random_8, children [WavePlayer_24, WavePlayer_25], properties {weights [1,1]}}
 {id=WavePlayer_10, properties {soundWaveAssetPtr=&/Tower_Aftermath_Vine_Flora_Pain_03}}
 {id=WavePlayer_11, properties {soundWaveAssetPtr=&/Tower_Aftermath_Vine_Flora_Pain_04}}
 {id=WavePlayer_12, properties {soundWaveAssetPtr=&/Tower_Aftermath_Vine_Flora_Pain_01}}
 {id=WavePlayer_13, properties {soundWaveAssetPtr=&Tower_Aftermath_Vine_Flora_Pain_04}}
 {id=WavePlayer_14, properties {soundWaveAssetPtr=&Tower_Aftermath_Vine_Flora_Pain_01}}
 {id=WavePlayer_15, properties {soundWaveAssetPtr=&Tower_Aftermath_Vine_Flora_Pain_04}}
 {id=WavePlayer_16, properties {soundWaveAssetPtr=&Tower_Aftermath_Vine_Flora_Pain_01}}
 {id=WavePlayer_17, properties {soundWaveAssetPtr=&/Tower_Aftermath_Vine_Flora_Impact_01}}
 {id=WavePlayer_18, properties {soundWaveAssetPtr=&/Tower_Aftermath_Vine_Flora_Impact_02}}
 {id=WavePlayer_19, properties {soundWaveAssetPtr=&/Tower_Aftermath_Vine_Flora_Impact_03}}
 {id=WavePlayer_20, properties {soundWaveAssetPtr=&/Tower_Aftermath_Vine_Flora_Impact_04}}
 {id=WavePlayer_21, properties {soundWaveAssetPtr=&Tower_Aftermath_Vine_Flora_Impact_01}}
 {id=WavePlayer_22, properties {soundWaveAssetPtr=&Tower_Aftermath_Vine_Flora_Impact_02}}
 {id=WavePlayer_23, properties {soundWaveAssetPtr=&Tower_Aftermath_Vine_Flora_Impact_04}}
 {id=WavePlayer_24, properties {soundWaveAssetPtr=&Tower_Aftermath_Vine_Flora_Impact_02}}
 {id=WavePlayer_25, properties {soundWaveAssetPtr=&Tower_Aftermath_Vine_Flora_Impact_04}}
 {id=WavePlayer_8, properties {soundWaveAssetPtr=&Tower_Aftermath_Vine_Flora_Pain_01}}
 {id=WavePlayer_9, properties {soundWaveAssetPtr=&/Tower_Aftermath_Vine_Flora_Pain_02}}
]
properties {
 attenuationSettings=/Game/Sounds/Attenuation/Fort_Demolition_Destroy_Small_Sized_Object_Att
 maxDistance=4000
 soundClassObject=&Fort_Impacts
 volumeMultiplier=0.68
 concurrencyOverrides {resolutionRule=PreventNew}
 concurrencySet [&CONC_Object_Impacts_FourMax]
}
