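# Sound cue for the death of an alien vine in the Tower Aftermath biome, played as a small-object destruction impact.
# A quality switch picks one of four mixes, with the first holding the most wave variations.
# Each layers a pitch-modulated impact with a briefly delayed second sound: death vox or pain with empty slots for occasional silence, or another impact in the last mix.
# An impact concurrency set, small-object attenuation and the impacts sound class govern playback.
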SoundCue /Game/Sounds/Apollo/Biomes/Tower_Aftermath/Tower_Aftermath_Alien_Vine_Death_Cue
root=QualityLevel_0
nodes [
 {id=Delay_0, children [Random_8], properties {delayMax=0.03, delayMin=0.02}}
 {id=Delay_1, children [Random_5], properties {delayMax=0.03, delayMin=0.02}}
 {id=Delay_2, children [Random_9], properties {delayMax=0.03, delayMin=0.02}}
 {id=Delay_3, children [Random_12], properties {delayMax=0.03, delayMin=0.02}}
 {id=Mixer_0, children [Modulator_0,Modulator_1], properties {inputVolume [1,1]}}
 {id=Mixer_1, children [Modulator_2,Modulator_4], properties {inputVolume [1,1]}}
 {id=Mixer_2, children [Modulator_3,Modulator_6], properties {inputVolume [1,1]}}
 {id=Mixer_3, children [Modulator_5,Modulator_7], properties {inputVolume [1,1]}}
 {id=Modulator_0, children [Random_4], properties {pitchMax=1.07, pitchMin=0.92}}
 {id=Modulator_1, children [Delay_0]}
 {id=Modulator_2, children [Random_10], properties {pitchMax=1.07, pitchMin=0.92}}
 {id=Modulator_3, children [Random_11], properties {pitchMax=1.07, pitchMin=0.92}}
 {id=Modulator_4, children [Delay_1]}
 {id=Modulator_5, children [WavePlayer_35], properties {pitchMax=1.1, pitchMin=0.9}}
 {id=Modulator_6, children [Delay_2]}
 {id=Modulator_7, children [Delay_3]}
 {id=QualityLevel_0, children [Mixer_0,Mixer_1,Mixer_2,Mixer_3]}
 {id=Random_10, children [WavePlayer_39, WavePlayer_40, WavePlayer_41], properties {weights [1,1,1]}}
 {id=Random_11, children [WavePlayer_42, WavePlayer_43, WavePlayer_44], properties {weights [1,1,1]}}
 {id=Random_12, children [WavePlayer_45, WavePlayer_46], properties {weights [1,1]}}
 {
  id=Random_4
  children [WavePlayer_15, WavePlayer_16, WavePlayer_17, WavePlayer_18, WavePlayer_19, WavePlayer_20, WavePlayer_21]
  properties {weights [1,1,1,1,1,1,1]}
 }
 {id=Random_5, children [WavePlayer_25, WavePlayer_26, WavePlayer_27, null, null], properties {weights [1,1,1,1,1]}}
 {
  id=Random_8
  children [WavePlayer_22, WavePlayer_23, WavePlayer_24, WavePlayer_31, WavePlayer_33, WavePlayer_34, null, null, null]
  properties {weights [1,1,1,1,1,1,1,1,1]}
 }
 {id=Random_9, children [WavePlayer_36, WavePlayer_37, WavePlayer_38, null, null], properties {weights [1,1,1,1,1]}}
 {id=WavePlayer_15, properties {soundWaveAssetPtr=&/Tower_Aftermath_Vine_Death_Impact_01}}
 {id=WavePlayer_16, properties {soundWaveAssetPtr=&/Tower_Aftermath_Vine_Death_Impact_02}}
 {id=WavePlayer_17, properties {soundWaveAssetPtr=&/Tower_Aftermath_Vine_Death_Impact_03}}
 {id=WavePlayer_18, properties {soundWaveAssetPtr=&/Tower_Aftermath_Vine_Death_Impact_04}}
 {id=WavePlayer_19, properties {soundWaveAssetPtr=&/Tower_Aftermath_Vine_Death_Impact_06}}
 {id=WavePlayer_20, properties {soundWaveAssetPtr=&/Tower_Aftermath_Vine_Death_Impact_05}}
 {id=WavePlayer_21, properties {soundWaveAssetPtr=&/Tower_Aftermath_Vine_Death_Impact_07}}
 {id=WavePlayer_22, properties {soundWaveAssetPtr=&/Tower_Aftermath_Vine_Death_Vox_01}}
 {id=WavePlayer_23, properties {soundWaveAssetPtr=&/Tower_Aftermath_Vine_Death_Vox_02}}
 {id=WavePlayer_24, properties {soundWaveAssetPtr=&/Tower_Aftermath_Vine_Death_Vox_03}}
 {id=WavePlayer_25, properties {soundWaveAssetPtr=&/Tower_Aftermath_Vine_Pain_01}}
 {id=WavePlayer_26, properties {soundWaveAssetPtr=&/Tower_Aftermath_Vine_Pain_03}}
 {id=WavePlayer_27, properties {soundWaveAssetPtr=&/Tower_Aftermath_Vine_Pain_02}}
 {id=WavePlayer_31, properties {soundWaveAssetPtr=&/Tower_Aftermath_Vine_Death_Vox_04}}
 {id=WavePlayer_33, properties {soundWaveAssetPtr=&/Tower_Aftermath_Vine_Death_Vox_05}}
 {id=WavePlayer_34, properties {soundWaveAssetPtr=&/Tower_Aftermath_Vine_Death_Vox_06}}
 {id=WavePlayer_35, properties {soundWaveAssetPtr=&Tower_Aftermath_Vine_Death_Impact_07}}
 {id=WavePlayer_36, properties {soundWaveAssetPtr=&Tower_Aftermath_Vine_Death_Vox_02}}
 {id=WavePlayer_37, properties {soundWaveAssetPtr=&Tower_Aftermath_Vine_Death_Vox_05}}
 {id=WavePlayer_38, properties {soundWaveAssetPtr=&Tower_Aftermath_Vine_Death_Vox_06}}
 {id=WavePlayer_39, properties {soundWaveAssetPtr=&Tower_Aftermath_Vine_Death_Impact_04}}
 {id=WavePlayer_40, properties {soundWaveAssetPtr=&Tower_Aftermath_Vine_Death_Impact_07}}
 {id=WavePlayer_41, properties {soundWaveAssetPtr=&Tower_Aftermath_Vine_Death_Impact_06}}
 {id=WavePlayer_42, properties {soundWaveAssetPtr=&Tower_Aftermath_Vine_Death_Impact_04}}
 {id=WavePlayer_43, properties {soundWaveAssetPtr=&Tower_Aftermath_Vine_Death_Impact_07}}
 {id=WavePlayer_44, properties {soundWaveAssetPtr=&Tower_Aftermath_Vine_Death_Impact_06}}
 {id=WavePlayer_45, properties {soundWaveAssetPtr=&Tower_Aftermath_Vine_Death_Impact_07}}
 {id=WavePlayer_46, properties {soundWaveAssetPtr=&Tower_Aftermath_Vine_Death_Impact_04}}
]
properties {
 attenuationSettings=/Game/Sounds/Attenuation/Fort_Demolition_Destroy_Small_Sized_Object_Att
 maxDistance=4000
 soundClassObject=&Fort_Impacts
 volumeMultiplier=0.68
 concurrencyOverrides {resolutionRule=PreventNew}
 concurrencySet [&CONC_Object_Impacts_FourMax]
}
