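# Impact sound cue for alien vines in the Tower Aftermath biome, using small-object demolition attenuation and impact concurrency.
# A Quality Level node picks one of four mixers, each layering an impact with a slightly delayed vine pain sound.
# Higher tiers use vine death impacts, while lower tiers use generic shrub impacts and fewer variations.
# Random picks and modulators add variety, and empty slots in the top tier's pain random sometimes leave that layer silent.
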SoundCue /Game/Sounds/Apollo/Biomes/Tower_Aftermath/Tower_Aftermath_Alien_Vine_Impact_Cue
root=QualityLevel_0
nodes [
 {id=Delay_0, children [Random_3], properties {delayMax=0.03, delayMin=0.02}}
 {id=Delay_1, children [Random_5], properties {delayMax=0.03, delayMin=0.02}}
 {id=Delay_2, children [Random_6], properties {delayMax=0.03, delayMin=0.02}}
 {id=Delay_3, children [Random_7], properties {delayMax=0.03, delayMin=0.02}}
 {id=Mixer_0, children [Modulator_0,Modulator_1], properties {inputVolume [1.25,0.75]}}
 {id=Mixer_1, children [Modulator_2,Modulator_4], properties {inputVolume [1,1]}}
 {id=Mixer_2, children [Modulator_3,Modulator_6], properties {inputVolume [1,0.75]}}
 {id=Mixer_3, children [Modulator_5,Modulator_7], properties {inputVolume [1,1]}}
 {id=Modulator_0, children [Random_4], properties {pitchMax=1.07, pitchMin=0.92}}
 {id=Modulator_1, children [Delay_0], properties {volumeMax=1.2, volumeMin=1}}
 {id=Modulator_2, children [Random_8], properties {pitchMax=1.07, pitchMin=0.92}}
 {id=Modulator_3, children [Random_2], properties {pitchMax=1.07, pitchMin=0.92}}
 {id=Modulator_4, children [Delay_1]}
 {id=Modulator_5, children [WavePlayer_7], properties {pitchMax=1.1, pitchMin=0.9}}
 {id=Modulator_6, children [Delay_2]}
 {id=Modulator_7, children [Delay_3]}
 {id=QualityLevel_0, children [Mixer_0,Mixer_1,Mixer_2,Mixer_3]}
 {id=Random_2, children [WavePlayer_5,WavePlayer_6], properties {weights [1,1]}}
 {
  id=Random_3
  children [WavePlayer_8, WavePlayer_9, WavePlayer_10, WavePlayer_11, WavePlayer_12, WavePlayer_14, null, null, null]
  properties {weights [1,1,1,1,1,1,1,1,1]}
 }
 {
  id=Random_4
  children [WavePlayer_15, WavePlayer_16, WavePlayer_17, WavePlayer_18, WavePlayer_19, WavePlayer_20, WavePlayer_21]
  properties {weights [1,1,1,1,1,1,1]}
 }
 {id=Random_5, children [WavePlayer_25, WavePlayer_26, WavePlayer_27], properties {weights [1,1,1]}}
 {id=Random_6, children [WavePlayer_28, WavePlayer_29], properties {weights [1,1]}}
 {id=Random_7, children [WavePlayer_32, WavePlayer_30], properties {weights [1,1]}}
 {id=Random_8, children [WavePlayer_22, WavePlayer_23], properties {weights [1,1]}}
 {id=WavePlayer_10, properties {soundWaveAssetPtr=&/Tower_Aftermath_Vine_Pain_03}}
 {id=WavePlayer_11, properties {soundWaveAssetPtr=&/Tower_Aftermath_Vine_Pain_04}}
 {id=WavePlayer_12, properties {soundWaveAssetPtr=&/Tower_Aftermath_Vine_Pain_05}}
 {id=WavePlayer_14, properties {soundWaveAssetPtr=&/Tower_Aftermath_Vine_Pain_07}}
 {id=WavePlayer_15, properties {soundWaveAssetPtr=&/Tower_Aftermath_Vine_Death_Impact_01}}
 {id=WavePlayer_16, properties {soundWaveAssetPtr=&/Tower_Aftermath_Vine_Death_Impact_02}}
 {id=WavePlayer_17, properties {soundWaveAssetPtr=&/Tower_Aftermath_Vine_Death_Impact_03}}
 {id=WavePlayer_18, properties {soundWaveAssetPtr=&/Tower_Aftermath_Vine_Death_Impact_04}}
 {id=WavePlayer_19, properties {soundWaveAssetPtr=&/Tower_Aftermath_Vine_Death_Impact_05}}
 {id=WavePlayer_20, properties {soundWaveAssetPtr=&/Tower_Aftermath_Vine_Death_Impact_06}}
 {id=WavePlayer_21, properties {soundWaveAssetPtr=&/Tower_Aftermath_Vine_Death_Impact_07}}
 {id=WavePlayer_22, properties {soundWaveAssetPtr=&Tower_Aftermath_Vine_Death_Impact_01}}
 {id=WavePlayer_23, properties {soundWaveAssetPtr=&Tower_Aftermath_Vine_Death_Impact_02}}
 {id=WavePlayer_25, properties {soundWaveAssetPtr=&/Tower_Aftermath_Vine_Pain_01}}
 {id=WavePlayer_26, properties {soundWaveAssetPtr=&Tower_Aftermath_Vine_Pain_03}}
 {id=WavePlayer_27, properties {soundWaveAssetPtr=&/Tower_Aftermath_Vine_Pain_02}}
 {id=WavePlayer_28, properties {soundWaveAssetPtr=&Tower_Aftermath_Vine_Pain_01}}
 {id=WavePlayer_29, properties {soundWaveAssetPtr=&Tower_Aftermath_Vine_Pain_02}}
 {id=WavePlayer_30, properties {soundWaveAssetPtr=&Tower_Aftermath_Vine_Pain_03}}
 {id=WavePlayer_32, properties {soundWaveAssetPtr=&Tower_Aftermath_Vine_Pain_04}}
 {id=WavePlayer_5, properties {soundWaveAssetPtr=/Game/Sounds/Fort_Impact_Sounds/Demolition/Bush/shrub_impact_01}}
 {id=WavePlayer_6, properties {soundWaveAssetPtr=/Game/Sounds/Fort_Impact_Sounds/Demolition/Bush/shrub_impact_02}}
 {id=WavePlayer_7, properties {soundWaveAssetPtr=&shrub_impact_01}}
 {id=WavePlayer_8, properties {soundWaveAssetPtr=&Tower_Aftermath_Vine_Pain_01}}
 {id=WavePlayer_9, properties {soundWaveAssetPtr=&Tower_Aftermath_Vine_Pain_02}}
]
properties {
 attenuationSettings=/Game/Sounds/Attenuation/Fort_Demolition_Hit_Small_Sized_Object_Att
 maxDistance=4000
 soundClassObject=&Fort_Impacts
 volumeMultiplier=0.68
 concurrencyOverrides {resolutionRule=PreventNew}
 concurrencySet [&CONC_Object_Impacts_FourMax]
}
